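# Unlit material in the MaterialGym folder that exercises the NewGalaxy material function, named for a triplanar cube case.
# MF_NewGalaxy_Generic receives a SetMaterialAttributes node with no values wired in, and its Result feeds a GetMaterialAttributes node that extracts a single attribute.
# The graph holds no triplanar logic or parameters of its own, so any galaxy effect comes from the shared function.
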Material /Game/Global/Materials/MaterialGym/Materials/NewGalaxyFunction/M_NewGalaxy_TriplanarCube
graph {
 properties {pixelDepthOffsetMode=PDOM_Legacy, shadingModel=MSM_Unlit, usedShadingModels=Unlit}
 nodes [GetMaterialAttributes(None=MaterialFunctionCall!(Material=SetMaterialAttributes!(MaterialAttributes), TechArtMask) {materialFunction=&/MF_NewGalaxy_Generic}.Result) {attributeGetTypes [#1]}]
}
